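# One-off editor utility actor that bulk-sets distance field lighting on the level's building actors from its construction script.
# It runs only in editor or PIE with no valid Athena game state, using the RunDisableFields boolean as a self-resetting trigger.
# It filters level actors to BuildingActor and writes bAffectDistanceFieldLighting on every primitive component through SetEditorProperty.
# Despite its name, it sets that property to true.
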EditorUtilityBlueprint /Game/Blueprints/Blutilities/Developers/brycelumpkin/OneOffs/DisableDistanceFields_EditorActor
graphs [
 EventGraph {nodes [
   Event BeginPlay()
   Event ActorBeginOverlap()
   Event Tick!() -> (OutputDelegate, then, DeltaSeconds:float)
  ]}
 UserConstructionScript {nodes [
   Construction Script()
   > runDisableDistanceFields()
  ]}
 runDisableDistanceFields {nodes [
   runDisableDistanceFields() {is=title K2Node_FunctionEntry}
   n1 = IsValid(Object=GetGameStateAthena!(WorldContextObject:Object Reference) -> (ReturnValue:Fort Game State Athena Object Reference))
   > Branch(Condition=AND Boolean(A=InEditorOrPIE!(WorldContextObject:Object Reference) -> (ReturnValue:bool), B=NOT Boolean(A=@n1)))
   > Branch(Condition=Get RunDisableFields() -> (RunDisableFields:bool))
   > Set RunDisableFields(RunDisableFields:bool=false) -> (Output_Get:bool)
   > n2 = GetAllLevelActors(self=Get EditorActorSubsystem())
   > n3 = Filter by Class!(
    execute
    TargetArray:Array of Object References=@n2
    ObjectClass:Object Class Reference=/Script/FortniteGame.BuildingActor
    FilterType:EEditorScriptingFilterType Enum
   ) -> (then, ReturnValue:Array of Building Actor Object References)
   > n4 = For Each Loop(Array:Array of Actor Object References=@n3) -> (Array Element:Actor Object Reference)
   > n5 = For Each Loop(Array:Array of Primitive Component Object References=Get Components By Class!(self:Actor Object Reference=@n4, ComponentClass:Actor Component Class Reference=/Script/Engine.PrimitiveComponent) -> (ReturnValue:Array of Primitive Component Object References)) -> (Array Element:Primitive Component Object Reference)
   > SetEditorProperty(Object=@n5, PropertyName=bAffectDistanceFieldLighting, PropertyValue=MakeLiteralBool(Value=true))
  ]}
]
variables [RunDisableFields {type=boolean}]
